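; Gameplay ability for a Season 23 narrative quest that spawns Loose Chip quest actors in response to shield hits.
; On activation it waits for Event.Shield.Hit gameplay events and checks their target tags for TargetFilter.AttachedActorAlwaysObstructsWorld.
; A matching hit has a roughly even random chance to call SpawnQuestActorsfromDropEntry with a single-chip drop entry, forwarding the event payload.
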(event Ability|EventActivateAbility
  (bind (_asynctaskproxy _payload) (Ability|Tasks|WaitGameplayEvent "(TagName=\"Event.Shield.Hit\")")
    (:EventReceived
      (bind (_payload_eventtag _payload_instigator _payload_target _payload_optionalobject _payload_optionalobject2 _payload_contexthandle _payload_instigatortags _payload_targettags _payload_eventmagnitude _payload_targetdata) (Utilities|Struct|BreakGameplayEventData _payload))
      (if (GameplayTags|HasTag _payload_targettags "(TagName=\"TargetFilter.AttachedActorAlwaysObstructsWorld\")")
        (if (<= (Math|Random|RandomFloat) 0.5)
          (Spawning|SpawnQuestActorsfromDropEntry 0 (struct :SpawnedQuestActorClass_6_F15CFDE2424412E9BADBE2975B317A07 "/BattlePassS23/QuestActors/Narrative/Part03/Quest_04/BP_S23_P03_Q04_Loose_Chip.BP_S23_P03_Q04_Loose_Chip_C" :DropQuantity_14_6469E6844F83772BB80669BCF263CB9D 1 :FailCount_10_49E24745476ACF5406DC15A79201F03E 1) (Utilities|Struct|MakeGameplayEventData _payload_eventtag _payload_instigator _payload_target _payload_optionalobject _payload_optionalobject2 _payload_contexthandle _payload_instigatortags _payload_targettags _payload_eventmagnitude _payload_targetdata)))))))

(event Ability|EventFailedToActivatePassiveAbility
  (params (ActorInfo :type Gameplay%20Ability%20Actor%20Info%20Structure)))
